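; Latent gameplay cue notify for hot air balloon destruction, adding color-matched burst particles and sound.
; On cue handling, it casts the target to BGA_BalloonHealth and caches that balloon's burst color.
; OnBurst applies the cached color to a spawned particle component's BalloonColor parameter and plays the configured sound attached to the target's root component.
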(event GameplayCueNotify|EventHandleGameplayCue
  (params (MyTarget :type Actor%20Object%20Reference) (EventType :type EGameplayCueEvent%20Enum :default "OnActive") (Parameters :type Gameplay%20Cue%20Parameters%20Structure%20%28by%20ref%29))
  (bind _asbga_balloon_health (Utilities|Casting|CastToBGA_BalloonHealth MyTarget))
  (Variables|Default|SetBurstColor (Variables|/HotAirBalloon/Gameplay/BGA_BalloonHealth.BGA_BalloonHealth_C|GetBurstColor _asbga_balloon_health)))

(fn OnBurst(EitherVFX-preferthisversion)
  (params (MyTarget :type Actor%20Object%20Reference) (Parameters :type Gameplay%20Cue%20Parameters%20Structure%20%28by%20ref%29) (ParticleComponents :type Array%20of%20FXSystem%20Component%20Object%20References) (AudioComponents :type Array%20of%20Audio%20Component%20Object%20References) (BurstCameraShakeInstance :type Camera%20Shake%20Base%20Object%20Reference) (BurstDecalInstance :type Decal%20Actor%20Object%20Reference))
  (Effects|Components|ParticleSystem|SetColorParameter (Utilities|Array|Get(acopy) ParticleComponents) "BalloonColor" (Variables|Default|GetBurstColor))
  (Audio|SpawnSoundAttached (Variables|Default|GetSound) (Variables|/Script/Engine.Actor|GetRootComponent MyTarget))
  (return))

(fn ConstructionScript)
